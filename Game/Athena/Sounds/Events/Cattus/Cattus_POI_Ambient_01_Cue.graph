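# Sound cue named as the first point-of-interest ambient for the Cattus event in Athena, assigned to the outdoor ambient sound class.
# It has no root or sound nodes, so the cue as shown produces no audio.
# Its overridden attenuation disables spatialization and sets a falloff beyond an inner radius, suggesting a non-directional area bed that fades with distance.
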SoundCue /Game/Athena/Sounds/Events/Cattus/Cattus_POI_Ambient_01_Cue
root=null
nodes []
properties {
 bOverrideAttenuation=true
 maxDistance=3100
 soundClassObject=&Fort_Ambient_Outdoors_Athena
 volumeMultiplier=1.002
 attenuationOverrides {bSpatialize=false, falloffDistance=1500, attenuationShapeExtents {x=1600}}
}
