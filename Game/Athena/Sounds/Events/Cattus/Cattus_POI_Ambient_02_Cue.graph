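# Sound cue named as the second ambient sound for a point of interest in the Cattus event, under Athena event sounds.
# It has no root or sound nodes, so the file defines only playback settings and no audio source.
# Those settings route it to the scripted events sound class at reduced volume and override attenuation with a capsule-shaped falloff.
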SoundCue /Game/Athena/Sounds/Events/Cattus/Cattus_POI_Ambient_02_Cue
root=null
nodes []
properties {
 bOverrideAttenuation=true
 maxDistance=6343.957
 soundClassObject=/Game/Sounds/FortSoundClassesAndModes/Fort_Events_Scripted
 volumeMultiplier=0.1754
 attenuationOverrides {attenuationShape=Capsule, falloffDistance=3000, attenuationShapeExtents {x=3843.957}}
}
