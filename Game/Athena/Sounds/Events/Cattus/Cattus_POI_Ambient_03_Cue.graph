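# Sound cue named as the third ambient sound for a Cattus point of interest in Athena, assigned to the outdoor ambient sound class.
# The cue has no root or sound nodes, so it plays nothing as saved and holds only settings.
# Its attenuation override turns off spatialization and fades volume over distance beyond an inner radius, with equal minimum and maximum reverb wet levels.
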SoundCue /Game/Athena/Sounds/Events/Cattus/Cattus_POI_Ambient_03_Cue
root=null
nodes []
properties {
 bOverrideAttenuation=true
 maxDistance=1280
 soundClassObject=&Fort_Ambient_Outdoors_Athena
 volumeMultiplier=1.002
 attenuationOverrides {bSpatialize=false, falloffDistance=1024, reverbWetLevelMax=10, reverbWetLevelMin=10, attenuationShapeExtents {x=256}}
}
